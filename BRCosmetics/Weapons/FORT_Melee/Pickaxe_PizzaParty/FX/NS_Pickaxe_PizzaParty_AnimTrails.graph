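# Niagara swing-trail effect for the Pizza Party pickaxe, built from three short-lived ribbon emitters: SoftDarkTrail, SoftGrey and TipWhiteTrail.
# Each emitter uses the shared AnimTrail functions to spawn and update ribbons from the system skeletal mesh, adding a user-supplied velocity.
# User.VariantPicker blends each emitter between two color-over-life curves to give style variants.
# TipWhiteTrail also applies time-of-day colors, including a blue multiplier at night.
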NiagaraSystem /BRCosmetics/Weapons/FORT_Melee/Pickaxe_PizzaParty/FX/NS_Pickaxe_PizzaParty_AnimTrails
dependencies {
 usedDataInterfaces [&NiagaraDataInterfaceColorCurve, &NiagaraDataInterfaceSkeletalMesh]
 usedDynamicInputs []
 usedModules [
  &self:SystemSpawnScript.SystemScriptSource.SystemScriptGraph.NiagaraNodeAssignment_0.SetVariables_#1
  &SystemState
 ]
}
emitters {
 SoftDarkTrail {
  emitterUpdateScript [
   &EmitterState()
   &FN_SpawnAnimTrail()
  ]
  particleSpawnScript [
   &FN_InitializeAnimTrailBoneIndecies()
   &FN_InitializeAnimTrail(Lifetime=0.1, SkeletalMesh=$System.SkeletalMesh)
   &AddVelocity(Velocity=$User.NewVector3f)
  ]
  particleUpdateScript [
   &FN_UpdateAnimTrail(SkeletalMesh=$System.SkeletalMesh) {module=FN_UpdateAnimTrail001}
   &Color(Color=&LerpLinearColors(
     StartColor=&ColorFromCurve(ColorCurve=DataInterface() {propertyValues {
        AlphaCurve {keys [<at 0 0.75 RCIM_Cubic>,<at 0.8981 0 RCIM_Cubic>]}
        BlueCurve {keys [<at 0 0.0392>]}
        GreenCurve {keys [<at 0 0.0441>]}
        RedCurve {keys [<at 0 0.1979>]}
       }})
     EndColor=&ColorFromCurve(ColorCurve=DataInterface() {propertyValues {
        AlphaCurve {keys [<at 0 0.15 RCIM_Cubic>,<at 0.8981 0 RCIM_Cubic>]}
        BlueCurve {keys [<at 0 0.7682>]}
        GreenCurve {keys [<at 0 0.6838>]}
        RedCurve {keys [<at 0 0.1379>]}
       }})
     LerpFactor=$User.VariantPicker
    ))
   &SolveForcesAndVelocity(Clamp Velocity=false, Limit Acceleration=false)
  ]
  rendererClasses [&NiagaraRibbonRendererProperties]
  renderers [{rendererClass=&NiagaraRibbonRendererProperties}]
 }
 SoftGrey {
  emitterUpdateScript [
   &EmitterState()
   &FN_SpawnAnimTrail()
  ]
  particleSpawnScript [
   &FN_InitializeAnimTrailBoneIndecies()
   &FN_InitializeAnimTrail(Lifetime=0.1, SkeletalMesh=$System.SkeletalMesh)
   &AddVelocity(Velocity=$User.NewVector3f)
  ]
  particleUpdateScript [
   &FN_UpdateAnimTrail(SkeletalMesh=$System.SkeletalMesh) {module=FN_UpdateAnimTrail001}
   &Color(Color=&LerpLinearColors(
     StartColor=&ColorFromCurve(ColorCurve=DataInterface() {propertyValues {
        AlphaCurve {keys [<at 0 0.4>,<at 0.6442 0 RCIM_Cubic>]}
        BlueCurve {keys [<at 0 0.8385>,<at 0.4212 0.1658>]}
        GreenCurve {keys [<at 0 0.695>,<at 0.4212 0.1659>]}
        RedCurve {keys [<at 0 0.6708>,<at 0.4212 0.1667>]}
       }})
     EndColor=&ColorFromCurve(ColorCurve=DataInterface() {propertyValues {
        AlphaCurve {keys [<at 0 1>,<at 0.6442 0 RCIM_Cubic>]}
        BlueCurve {keys [<at 0 0.2448>,<at 0.4212 0.1658>]}
        GreenCurve {keys [<at 0 0.1511>,<at 0.4212 0.1659>]}
        RedCurve {keys [<at 0 0>,<at 0.4212 0.1667>]}
       }})
     LerpFactor=$User.VariantPicker
    ))
   &SolveForcesAndVelocity(Clamp Velocity=false, Limit Acceleration=false)
  ]
  rendererClasses [&NiagaraRibbonRendererProperties]
  renderers [{rendererClass=&NiagaraRibbonRendererProperties}]
 }
 TipWhiteTrail {
  emitterUpdateScript [
   &EmitterState()
   &FN_SpawnAnimTrail(AnimTrailWidthMode=::FromSecond)
  ]
  particleSpawnScript [
   &FN_InitializeAnimTrailBoneIndecies()
   &FN_InitializeAnimTrail(
    AnimTrailWidth=0.1
    Color=&FN_Select_Color_from_Time_Of_Day(Morning=<rgba 3.8565 0 4 1>, Day=<rgba 5 0 4.2726 1>, Evening=<rgba 4 0 3.8034 1>, Night=<rgba 5 1.7792 0 1>)
    Lifetime=0.1
    SkeletalMesh=$System.SkeletalMesh
   )
   &AddVelocity(Velocity=$User.NewVector3f)
  ]
  particleUpdateScript [
   &FN_UpdateAnimTrail(SkeletalMesh=$System.SkeletalMesh) {module=FN_UpdateAnimTrail001}
   &Color(Color=&LerpLinearColors(
     StartColor=&ColorFromCurve(ColorCurve=DataInterface() {propertyValues {AlphaCurve {keys [<at 0 1>,<at 0.2175 0 RCIM_Cubic>]}, BlueCurve {keys [<at 0 1>]}, GreenCurve {keys [<at 0 1>]}, RedCurve {keys [<at 0 1>]}}})
     EndColor=&ColorFromCurve(ColorCurve=DataInterface() {propertyValues {
        AlphaCurve {keys [<at 0 1>,<at 0.2175 0 RCIM_Cubic>]}
        BlueCurve {keys [<at 0 0.7991>]}
        GreenCurve {keys [<at 0 0.3419>]}
        RedCurve {keys [<at 0 0>]}
       }})
     LerpFactor=$User.VariantPicker
    ))
   /SetVariables(Particles.Color=&MultiplyLinearColor(
     Linear Color=&MultiplyLinearColor(
      Linear Color=$Particles.Color
      Scale Factor=&MakeVector4_FromLinearColor(LinearColor=&FNTimeOfDayColor(Night Color=<rgba 0 0.0603 3 1>, Morning Color=<rgba 1 1 1 1>, Day Color=<rgba 1 1 1 1>, Evening Color=<rgba 1 1 1 1>))
     )
     Scale Factor=&MakeVector4(X=1, Y=1, Z=1, W=1)
    )) {isSetParameters=true}
   &SolveForcesAndVelocity(Clamp Velocity=false, Limit Acceleration=false)
  ]
  rendererClasses [&NiagaraRibbonRendererProperties]
  renderers [{rendererClass=&NiagaraRibbonRendererProperties}]
 }
}
summary {
 systemName=NS_Pickaxe_PizzaParty_AnimTrails
 emitters [
  {emitterName=SoftDarkTrail, rendererClasses [&NiagaraRibbonRendererProperties]}
  {emitterName=SoftGrey, rendererClasses [&NiagaraRibbonRendererProperties]}
  {emitterName=TipWhiteTrail, rendererClasses [&NiagaraRibbonRendererProperties]}
 ]
 userVariables [
  User.Color A {type=LinearColor, description="", defaultValue {struct=LinearColor, value=<rgba 1 1 1 1>}}
  User.NewVector3f {type=Vector3f, description="", defaultValue {struct=Vector3f, value=<0 0 0>}}
  User.RibbonIDOffset {type=NiagaraInt32}
  User.ToonRibbonPanSpeed {type=NiagaraFloat, defaultValue=-2, description=""}
  User.VariantPicker {type=NiagaraFloat, defaultValue=0, description=""}
 ]
}
